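; BP_ColorRefPost is a post-process actor that shows a configurable reference texture through a material, apparently as a color reference test tool.
; Its construction script creates a dynamic instance of M_ColorRefTest, feeding it the Texture variable and an Offset vector built from Offset X and Y.
; It sets that instance as a full-weight blendable on its PostProcess component, then toggles the component with the Enabled variable.
; BeginPlay, Tick and overlap events are empty.
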(event EventBeginPlay)

(event Collision|EventActorBeginOverlap
  (params (OtherActor :type Actor%20Object%20Reference)))

(event EventTick
  (params (DeltaSeconds :type Float%20%28single-precision%29)))

(fn ConstructionScript
  (bind _postprocess (Variables|Default|GetPostProcess))
  (bind _returnvalue (Rendering|Material|CreateDynamicMaterialInstance "/WildEstate/NeverCook/M_ColorRefTest.M_ColorRefTest" "None" "None" :!class /Script/Engine.KismetMaterialLibrary))
  (Rendering|Material|SetTextureParameterValue _returnvalue "Tex" (Variables|Default|GetTexture))
  (Rendering|Material|SetVectorParameterValue _returnvalue "Offset" (Utilities|Struct|MakeLinearColor (Variables|Default|GetOffset%20X) (Variables|Default|GetOffset%20Y)) :!class /Script/Engine.MaterialInstanceDynamic)
  (Variables|/Script/Engine.PostProcessComponent|SetSettings (Utilities|Struct|MakePostProcessSettings (Utilities|Struct|MakeWeightedBlendables (Utilities|Array|MakeArray (Utilities|Struct|MakeWeightedBlendable 1.0 _returnvalue)))) _postprocess)
  (Variables|/Script/Engine.PostProcessComponent|SetbEnabled (Variables|Default|GetEnabled) _postprocess))
